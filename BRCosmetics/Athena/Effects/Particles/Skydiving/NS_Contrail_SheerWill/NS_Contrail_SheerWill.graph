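# Niagara system for the SheerWill skydiving contrail cosmetic, layering energy meshes, floating fragments, dark trails, hand glows and lightning sprites.
# Mostly persistent burst-spawned particles follow sampled skeletal bones or bob on waveforms, while lightning spawns continuously.
# A smoothed system value drives trail opacity and lightning intensity, and ParaGlideLeanAlpha scales fragments.
# VariantColorIndex selects Blot, Citadel or Froth looks, with FrontEnd modifiers adjusting size and brightness outside gameplay.
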NiagaraSystem /BRCosmetics/Athena/Effects/Particles/Skydiving/NS_Contrail_SheerWill/NS_Contrail_SheerWill
dependencies {
 usedDataInterfaces [
  &NiagaraDataInterfaceArrayFloat
  &NiagaraDataInterfacePlatformSet
  &NiagaraDataInterfaceArrayFloat3
  &NiagaraDataInterfaceSpriteRendererInfo
  &NiagaraDataInterfaceCurve
  &NiagaraDataInterfaceSkeletalMesh
 ]
 usedDynamicInputs [
  &self:ScratchDynamicInput_02
  &self:ScratchDynamicInput_03
 ]
 usedModules [
  &self:SystemSpawnScript.SystemScriptSource.SystemScriptGraph.NiagaraNodeAssignment_0.SetVariables_#1
  &SystemState
  &self:SystemSpawnScript.SystemScriptSource.SystemScriptGraph.NiagaraNodeAssignment_1.SetVariables_#2
  &self:SystemSpawnScript.SystemScriptSource.SystemScriptGraph.NiagaraNodeAssignment_2.SetVariables_#3
  &self:SystemSpawnScript.SystemScriptSource.SystemScriptGraph.NiagaraNodeAssignment_3.SetVariables_#4
  &self:SystemSpawnScript.SystemScriptSource.SystemScriptGraph.NiagaraNodeAssignment_4.SetVariables_#5
  &self:SystemSpawnScript.SystemScriptSource.SystemScriptGraph.NiagaraNodeAssignment_5.SetVariables_#6
  &self:SystemSpawnScript.SystemScriptSource.SystemScriptGraph.NiagaraNodeAssignment_6.SetVariables_#7
 ]
}
emitters {
 BottomEnergy {
  emitterUpdateScript [
   /SetVariables(Emitter.Des:NiagaraFloat=&self:ScratchDynamicInput_12(
     Blot:NiagaraFloat=0.5
     Citadel:NiagaraFloat=0.5
     Froth:NiagaraFloat=1
     VariantColorIndex:NiagaraInt32=&MakeIntFromFloat(Float=$User.VariantColorIndex)
    )) {isSetParameters=true}
   /SetVariables(Emitter.Des02:NiagaraFloat=&self:ScratchDynamicInput_12(
     Blot:NiagaraFloat=-0.2
     Citadel:NiagaraFloat=0.5
     Froth:NiagaraFloat=1
     VariantColorIndex:NiagaraInt32=&MakeIntFromFloat(Float=$User.VariantColorIndex)
    )) {isSetParameters=true}
   &EmitterState(Life Cycle Mode=::NewEnumerator1, Loop Behavior=::NewEnumerator1, Loop Duration Mode=::NewEnumerator1, UseLoopDelay=false)
   &SpawnBurst_Instantaneous(Use Spawn Probability=false, UseLoopCountLimit=false)
  ]
  particleSpawnScript [
   &InitializeParticle(
    Color Mode=::NewEnumerator0
    Mesh Scale Mode=::NewEnumerator3
    Mesh Uniform Scale=2
    Position Offset=<0 15 0>
    UsePositionOffset=true
    Write Material Random=true
   )
   &InitialMeshOrientation(Mesh Orientation Mode=::NewEnumerator7, Use Rotation Vector=true)
   &DynamicMaterialParameters(
    Index 0 Param 1=3
    Index 0 Param 2=0.4
    Index 2 Param 1=6
    Module.Index0Float0Write=true
    Module.Index0Float1Write=true
    Module.Index0Float2Write=true
    Module.Index0Float3Write=true
    Module.Index2Float0Write=true
    Module.Index2Float1Write=true
    Module.Index2Float2Write=true
    Module.Index2Float3Write=true
    Module.Index3Float0Write=true
    Module.Index3Float1Write=true
    Module.Index3Float2Write=true
    Module.Index3Float3Write=true
    Write Parameter Index 2=::NewEnumerator1
    Write Parameter Index 3=::NewEnumerator1
   )
   &CameraOffset(Camera Offset Amount=-10)
   /SetVariables(Particles.SizeAll:NiagaraFloat=&FN_FrontEndFloatModifier(Gameplay=1, Front End=0.9)) {isSetParameters=true}
  ]
  particleUpdateScript [
   /SetVariables(Particles.AlphaAll:NiagaraFloat=&Multiply_Float(A=0.7, B=&Lerp_Float(A=0, B=1, Alpha=$System.InterpolateOverTime.InterpolateOverTime_Float.Moving Average))) {isSetParameters=true}
   &ParticleState(Kill Particles When Lifetime Has Elapsed=false)
   &ScaleMeshSize(Scale Factor=&Multiply_VectorByFloat(Vector=&Multiply_VectorByFloat(Vector=<1.1 1.1 0.5>, Float=$Particles.SizeAll), Float=0.9))
   &FN_Mesh_Attributes(
    Color=&MakeLinearColorFromVectorAndFloat(
     "Vector (RGB)"=&Multiply_Vector(
      A=&MakeVectorFromLinearColorRGB(Color=&self:ScratchDynamicInput_01(
        BlotColor:LinearColor=<rgba 0.8 0.0594 0.0594 0.1>
        CitadelColor:LinearColor=<rgba 0.1114 0.5743 1.5 0.1>
        FrothColor:LinearColor=<rgba 0 0 0 1>
        VariantColorIndex:NiagaraInt32=&MakeIntFromFloat(Float=$User.VariantColorIndex)
       ))
      B=&Multiply_VectorByFloat(Vector=<1 1 1>, Float=3)
     )
     "Float (Alpha)"=&Multiply_Float(A=0.5, B=$Particles.AlphaAll)
    )
    Create Color Attribute=true
    Module Output=::NewEnumerator1
   ) {module=FN_Mesh_Attributes002}
   &FN_Mesh_Attributes(
    Color=&MakeLinearColorFromVectorAndFloat(
     "Vector (RGB)"=&Multiply_Vector(
      A=&MakeVectorFromLinearColorRGB(Color=&self:ScratchDynamicInput_01(
        BlotColor:LinearColor=<rgba 0.1 0 0.0008 0.1>
        CitadelColor:LinearColor=<rgba 0 0.0667 0.2 0.1>
        FrothColor:LinearColor=<rgba 0 0 0 1>
        VariantColorIndex:NiagaraInt32=&MakeIntFromFloat(Float=$User.VariantColorIndex)
       ))
      B=&Multiply_VectorByFloat(Vector=<1 1 1>, Float=5)
     )
     "Float (Alpha)"=&Multiply_Float(
      A=&Multiply_Float(
       A=0.7
       B=&self:ScratchDynamicInput_13(
        Blot:NiagaraFloat=1
        Citadel:NiagaraFloat=1
        Froth:NiagaraFloat=0
        VariantColorIndex:NiagaraInt32=&MakeIntFromFloat(Float=$User.VariantColorIndex)
       )
      )
      B=$Particles.AlphaAll
     )
    )
    Create Color Attribute=true
    Create Scale Attribute=true
    Module Output=::NewEnumerator1
    Scale=&Multiply_VectorByFloat(Vector=&Multiply_VectorByFloat(Vector=<1 1 1>, Float=$Particles.SizeAll), Float=3)
   ) {module=FN_Mesh_Attributes001}
   &FN_Mesh_Attributes(
    Color=&MakeLinearColorFromVectorAndFloat(
     "Vector (RGB)"=&Multiply_Vector(
      A=&MakeVectorFromLinearColorRGB(Color=&self:ScratchDynamicInput_01(
        BlotColor:LinearColor=<rgba 0.1 0 0.0008 0.1>
        CitadelColor:LinearColor=<rgba 1 1 1 0.1>
        FrothColor:LinearColor=<rgba 0 0 0 1>
        VariantColorIndex:NiagaraInt32=&MakeIntFromFloat(Float=$User.VariantColorIndex)
       ))
      B=&Multiply_VectorByFloat(Vector=<1 1 1>, Float=1)
     )
     "Float (Alpha)"=&Multiply_Float(A=0.5, B=$Particles.AlphaAll)
    )
    Create Color Attribute=true
    Create Scale Attribute=true
    Module Output=::NewEnumerator1
    Scale=&Multiply_VectorByFloat(Vector=&Multiply_VectorByFloat(Vector=<1 1 1>, Float=$Particles.SizeAll), Float=2.5)
   )
   &FN_Mesh_Attributes(
    Color=&MakeLinearColorFromVectorAndFloat(
     "Vector (RGB)"=&Multiply_Vector(
      A=&MakeVectorFromLinearColorRGB(Color=&self:ScratchDynamicInput_01(
        BlotColor:LinearColor=<rgba 0.1 0.0089 0.0096 0.1>
        CitadelColor:LinearColor=<rgba 0.009 0.0393 0.1 0.1>
        FrothColor:LinearColor=<rgba 0 0 0 1>
        VariantColorIndex:NiagaraInt32=&MakeIntFromFloat(Float=$User.VariantColorIndex)
       ))
      B=&Multiply_VectorByFloat(Vector=<1 1 1>, Float=12)
     )
     "Float (Alpha)"=&Multiply_Float(A=1, B=$Particles.AlphaAll)
    )
    Create Color Attribute=true
    Create Scale Attribute=true
    Module Output=::NewEnumerator1
    Scale=&Multiply_VectorByFloat(Vector=<1 1 0.5>, Float=&Multiply_Float(A=1.7, B=$Particles.SizeAll))
   ) {module=FN_Mesh_Attributes003}
   &FN_Mesh_Attributes(
    Color=&MakeLinearColorFromVectorAndFloat(
     "Vector (RGB)"=&Multiply_Vector(
      A=&MakeVectorFromLinearColorRGB(Color=&self:ScratchDynamicInput_01(
        BlotColor:LinearColor=<rgba 0.1 0.0089 0.0096 0.1>
        CitadelColor:LinearColor=<rgba 0.009 0.0393 0.1 0.1>
        FrothColor:LinearColor=<rgba 0 0 0 1>
        VariantColorIndex:NiagaraInt32=&MakeIntFromFloat(Float=$User.VariantColorIndex)
       ))
      B=&Multiply_VectorByFloat(Vector=<1 1 1>, Float=10)
     )
     "Float (Alpha)"=&Multiply_Float(A=1, B=$Particles.AlphaAll)
    )
    Create Color Attribute=true
    Create Scale Attribute=true
    Module Output=::NewEnumerator1
    Scale=&Multiply_VectorByFloat(Vector=<1 1 0.5>, Float=&Multiply_Float(A=1.5, B=$Particles.SizeAll))
   ) {module=FN_Mesh_Attributes005}
  ]
  rendererClasses [&NiagaraMeshRendererProperties]
  renderers [
   {rendererClass=&NiagaraMeshRendererProperties, rendererIndex=0}
   {rendererClass=&NiagaraMeshRendererProperties, rendererIndex=1}
   {rendererClass=&NiagaraMeshRendererProperties, rendererIndex=2}
   {rendererClass=&NiagaraMeshRendererProperties, rendererIndex=3}
   {rendererClass=&NiagaraMeshRendererProperties, rendererIndex=4}
  ]
 }
 BottomFragment {
  emitterUpdateScript [
   /SetVariables(Emitter.CrackColor:LinearColor=<rgba 1.0736 2.8405 5 1>) {isSetParameters=true}
   /SetVariables(Emitter.FresnelEdgeColor:LinearColor=<rgba 1.1782 3.9343 5 1>) {isSetParameters=true}
   &EmitterState(Life Cycle Mode=::NewEnumerator1, Loop Behavior=::NewEnumerator1, Loop Duration Mode=::NewEnumerator1, UseLoopDelay=false)
   &SpawnBurst_Instantaneous(Use Spawn Probability=false, UseLoopCountLimit=false)
  ]
  particleSpawnScript [&InitializeParticle(
    Color Mode=::NewEnumerator0
    Mesh Scale Mode=::NewEnumerator3
    Mesh Uniform Scale=2
    Position Offset=<0 15 10>
    UsePositionOffset=true
    Write Material Random=true
   )]
  particleUpdateScript [
   /SetVariables(Particles.RandomWave:NiagaraFloat=&Waveform(
     "Waveform [1]"=::NewEnumerator0
     "[1] Frequency"=&RandomRangeFloat(Minimum=0.5, Maximum=1)
     "[1] Amplitude Min/Max"=<0 1>
    )) {isSetParameters=true}
   &ParticleState(Kill Particles When Lifetime Has Elapsed=false)
   &InterpolateOverTime(
    Rate Of Change=8
    Target Float Value:NiagaraFloat=&MakeCustomFloatFromBool(Bool=&SetBoolByFloatComparison(A=$User.ParaGlideLeanAlpha, B=0.7), True Float=1, False Float=0)
   )
   &ScaleMeshSize(Scale Factor=&Multiply_VectorByFloat(
     Vector=<1 1 1>
     Float=&Lerp_Float(
      A=0
      B=&MakeCustomFloatFromBool(
       Bool=&SetBoolByFloatComparison(A=$Particles.InterpolateOverTime.InterpolateOverTime_Float.Moving Average, B=0.1, Comparison Type=::NewEnumerator6)
       True Float=0
       False Float=1.2
      )
      Alpha=$Particles.InterpolateOverTime.InterpolateOverTime_Float.Moving Average
     )
    ))
   /SetVariables(Particles.Position01:NiagaraPosition=&AddVectorToPosition(
     Position=$Particles.Initial.Position
     Vector=&Lerp_Vector(
      A=<0 0 0>
      B=<0 7 -7>
      Alpha=&Waveform(
       "Waveform [1]"=::NewEnumerator0
       "[1] Phase Scale"=&RandomRangeFloat(Minimum=0.5, Maximum=1)
       "[1] Frequency"=&RandomRangeFloat(Minimum=0.5, Maximum=1)
       "[1] Amplitude Min/Max"=<0 1.5>
      )
     )
    )) {isSetParameters=true}
   /SetVariables(Particles.Position02:NiagaraPosition=&AddVectorToPosition(
     Position=$Particles.Initial.Position
     Vector=&Lerp_Vector(
      A=<0 0 0>
      B=<0 -7 -7>
      Alpha=&Waveform(
       Global Amplitude Scale=&RandomRangeFloat(Minimum=0.5, Maximum=1)
       "Waveform [1]"=::NewEnumerator1
       "[1] Phase Scale"=&RandomRangeFloat(Minimum=0.5, Maximum=1)
       "[1] Frequency"=&RandomRangeFloat(Minimum=0.5, Maximum=1)
       "[1] Amplitude Min/Max"=<0 1.5>
      )
     )
    )) {isSetParameters=true}
   /SetVariables(Particles.Position03:NiagaraPosition=&AddVectorToPosition(
     Position=$Particles.Initial.Position
     Vector=&Lerp_Vector(
      A=<0 0 0>
      B=<0 7 -7>
      Alpha=&Waveform(
       Global Amplitude Scale=&RandomRangeFloat(Minimum=0.5, Maximum=1)
       "Waveform [1]"=::NewEnumerator0
       "[1] Phase Scale"=&RandomRangeFloat(Minimum=0.5, Maximum=1)
       "[1] Frequency"=&RandomRangeFloat(Minimum=0.5, Maximum=1)
       "[1] Amplitude Min/Max"=<0 1.5>
      )
     )
    )) {isSetParameters=true}
   /SetVariables(Particles.Position04:NiagaraPosition=&AddVectorToPosition(
     Position=$Particles.Initial.Position
     Vector=&Lerp_Vector(
      A=<0 0 0>
      B=<0 -7 -7>
      Alpha=&Waveform(
       Global Amplitude Scale=&RandomRangeFloat(Minimum=0.5, Maximum=1)
       "Waveform [1]"=::NewEnumerator1
       "[1] Phase Scale"=&RandomRangeFloat(Minimum=0.5, Maximum=1)
       "[1] Frequency"=&RandomRangeFloat(Minimum=0.5, Maximum=1)
       "[1] Amplitude Min/Max"=<0 1.5>
      )
     )
    )) {isSetParameters=true}
   &CameraOffset(Camera Offset Amount=-10)
  ]
  rendererClasses [&NiagaraMeshRendererProperties]
  renderers [
   {rendererClass=&NiagaraMeshRendererProperties, rendererIndex=0}
   {rendererClass=&NiagaraMeshRendererProperties, rendererIndex=1}
   {rendererClass=&NiagaraMeshRendererProperties, rendererIndex=2}
   {rendererClass=&NiagaraMeshRendererProperties, rendererIndex=3}
   {rendererClass=&NiagaraMeshRendererProperties, rendererIndex=4}
   {rendererClass=&NiagaraMeshRendererProperties, rendererIndex=5}
   {rendererClass=&NiagaraMeshRendererProperties, rendererIndex=6}
   {rendererClass=&NiagaraMeshRendererProperties, rendererIndex=7}
   {rendererClass=&NiagaraMeshRendererProperties, rendererIndex=8}
   {rendererClass=&NiagaraMeshRendererProperties, rendererIndex=9}
   {rendererClass=&NiagaraMeshRendererProperties, rendererIndex=10}
   {rendererClass=&NiagaraMeshRendererProperties, rendererIndex=11}
  ]
 }
 DarkTrail {
  emitterUpdateScript [
   &EmitterState(
    Inactive Response=::NewEnumerator1
    Life Cycle Mode=::NewEnumerator1
    Loop Behavior=::NewEnumerator1
    Loop Duration Mode=::NewEnumerator1
    UseLoopDelay=false
   )
   &SpawnBurst_Instantaneous(Spawn Count=2, Use Spawn Probability=false, UseLoopCountLimit=false)
  ]
  particleSpawnScript [
   &PartitionParticles()
   &InitializeParticle(Color Mode=::NewEnumerator0, Mesh Scale Mode=::NewEnumerator3, Mesh Uniform Scale=3.5, UsePositionOffset=false, Write Material Random=true)
   &InitialMeshOrientation(Mesh Orientation Mode=::NewEnumerator7, Rotation=<0 -0.25 0>, Use Rotation Vector=true)
   &DynamicMaterialParameters(
    Index 0 Param 1=&FN_FrontEndFloatModifier(Gameplay=1.4, Front End=1.2)
    Index 0 Param 2=&SelectFloatFromArray(Direct Array Index=$Particles.PartitionParticles.Partition, Float Selection Array=DataInterface() {propertyValues {FloatData [0.3,0.5]}})
    Module.Index0Float0Write=true
    Module.Index0Float1Write=true
    Module.Index0Float2Write=true
    Module.Index0Float3Write=true
   )
   &CameraOffset(Camera Offset Amount=-20)
  ]
  particleUpdateScript [
   &Color(Color=&self:ScratchDynamicInput_05(
     BlotColor:LinearColor=<rgba 1 0 0 1>
     CitadelColor:LinearColor=<rgba 0 0.3333 1 1>
     FrothColor:LinearColor=<rgba 0 0 0 1>
     VariantColorIndex:NiagaraInt32=&MakeIntFromFloat(Float=$User.VariantColorIndex)
    ))
   &ParticleState(Kill Particles When Lifetime Has Elapsed=false)
   &ScaleMeshSize(Scale Factor=&Multiply_VectorByFloat(
     Vector=<1 1 1>
     Float=&Lerp_Float(A=0.7, B=1, Alpha=$System.InterpolateOverTime.InterpolateOverTime_Float.Moving Average)
    ))
   &SkeletalMeshLocation(
    Bone / Socket Index=$Particles.PartitionParticles.Partition
    Bone Sampling Mode=::NewEnumerator4
    MaskSpawnGroup=false
    Orientation Sampling=::NewEnumerator1
    Sampled Position Offset=&FN_FrontEndVector3Modifier(
     Gameplay=&SelectVectorFromArray(
      Direct Array Index=$Particles.PartitionParticles.Partition
      Vector Selection Array=DataInterface() {propertyValues {FloatData [<15 0 0>, <-15 0 10>]}}
     )
     Front End=&SelectVectorFromArray(
      Direct Array Index=$Particles.PartitionParticles.Partition
      Vector Selection Array=DataInterface() {propertyValues {FloatData [<10 0 0>, <-10 0 10>]}}
     )
    )
    Sampled Position Offset Coordinate Space:ENiagaraSkelMeshTransforms=::NewEnumerator2
    Skeletal Mesh=$System.Skeletal Mesh Reader
   )
   /SetVariables(Particles.AlphaAll:NiagaraFloat=&Multiply_Float(A=1, B=$System.InterpolateOverTime.InterpolateOverTime_Float.Moving Average)) {isSetParameters=true}
   &ScaleColor(
    Color Value To Scale=$Particles.Color
    Scale Alpha=&Multiply_Float(A=$Particles.AlphaAll, B=&FN_FrontEndFloatModifier(Gameplay=1, Front End=0.4))
    ScaleA=true
    ScaleRGB=true
   )
   &FN_Mesh_Attributes(
    Color=&MakeLinearColorFromVectorAndFloat(
     "Vector (RGB)"=&Multiply_Vector(
      A=&MakeVectorFromLinearColorRGB(Color=&self:ScratchDynamicInput_06(
        BlotColor:LinearColor=<rgba 0 0 0 1>
        CitadelColor:LinearColor=<rgba 0 0.0167 0.1 1>
        FrothColor:LinearColor=<rgba 0 0 0 1>
        VariantColorIndex:NiagaraInt32=&MakeIntFromFloat(Float=$User.VariantColorIndex)
       ))
      B=<1 1 1>
     )
     "Float (Alpha)"=&Multiply_Float(
      A=$Particles.AlphaAll
      B=&self:ScratchDynamicInput_07(
       Blot:NiagaraFloat=1
       Citadel:NiagaraFloat=&FN_FrontEndFloatModifier(Gameplay=0.9, Front End=1)
       Froth:NiagaraFloat=1
       VariantColorIndex:NiagaraInt32=&MakeIntFromFloat(Float=$User.VariantColorIndex)
      )
     )
    )
    Create Color Attribute=true
    Module Output=::NewEnumerator1
   )
  ]
  rendererClasses [&NiagaraMeshRendererProperties]
  renderers [
   {rendererClass=&NiagaraMeshRendererProperties, rendererIndex=0}
   {rendererClass=&NiagaraMeshRendererProperties, rendererIndex=1}
  ]
 }
 HandCone {
  emitterUpdateScript [
   /SetVariables(Emitter.Des:NiagaraFloat=&self:ScratchDynamicInput_09(
     Blot:NiagaraFloat=0.5
     Citadel:NiagaraFloat=0.5
     Froth:NiagaraFloat=1
     VariantColorIndex:NiagaraInt32=&MakeIntFromFloat(Float=$User.VariantColorIndex)
    )) {isSetParameters=true}
   &EmitterState(Life Cycle Mode=::NewEnumerator1, Loop Behavior=::NewEnumerator1, Loop Duration Mode=::NewEnumerator1, UseLoopDelay=false)
   &SpawnBurst_Instantaneous(Spawn Count=2, Use Spawn Probability=false, UseLoopCountLimit=false)
  ]
  particleSpawnScript [
   &PartitionParticles()
   &InitializeParticle(
    Color Mode=::NewEnumerator0
    Mesh Scale Mode=::NewEnumerator3
    Mesh Uniform Scale=0.5
    Sprite Size Mode=::NewEnumerator3
    Uniform Sprite Size=80
    UsePositionOffset=false
    Write Material Random=true
   )
   &InitialMeshOrientation(Mesh Orientation Mode=::NewEnumerator7, Use Rotation Vector=true)
   &DynamicMaterialParameters(
    Index 0 Param 1=3
    Index 0 Param 2=1
    Module.Index0Float0Write=true
    Module.Index0Float1Write=true
    Module.Index0Float2Write=true
    Module.Index0Float3Write=true
   )
   &FN_Mesh_Attributes(
    Color=&MakeLinearColorFromVectorAndFloat(
     "Vector (RGB)"=&Multiply_Vector(A=&MakeVectorFromLinearColorRGB(Color=<rgba 1 0.0146 0 0.4>), B=<1 1 1>)
     "Float (Alpha)"=&Multiply_Float(A=0.3, B=1)
    )
    Create Color Attribute=true
    Lifetime=1
    Module Output=::NewEnumerator1
   )
   &CameraOffset(Camera Offset Amount=-20)
  ]
  particleUpdateScript [
   &Color(Color=&self:ScratchDynamicInput_05(
     BlotColor:LinearColor=<rgba 1 0 0 0.9>
     CitadelColor:LinearColor=<rgba 1.9532 2 1.034 0.9>
     FrothColor:LinearColor=<rgba 1 1 1 1>
     VariantColorIndex:NiagaraInt32=&MakeIntFromFloat(Float=$User.VariantColorIndex)
    ))
   &ParticleState(Kill Particles When Lifetime Has Elapsed=false)
   &ScaleMeshSize(Scale Factor=&Multiply_VectorByFloat(
     Vector=&Multiply_VectorByFloat(
      Vector=<0.8 0.8 0.8>
      Float=&RandomRangeFloat(Minimum=1, Maximum=1.1, Evaluation Type=::NewEnumerator1)
     )
     Float=0.9
    ))
   &SkeletalMeshLocation(
    Bone / Socket Index=$Particles.PartitionParticles.Partition
    Bone Sampling Mode=::NewEnumerator4
    MaskSpawnGroup=false
    Orientation Sampling=::NewEnumerator1
    Sampled Position Offset=&AddVectorToPosition(
     Position=<0 0 -10>
     Vector=&SelectVectorFromArray(
      Direct Array Index=$Particles.PartitionParticles.Partition
      Vector Selection Array=DataInterface() {propertyValues {FloatData [<2 3 0>, <-2 3 0>]}}
     )
    )
    Sampled Position Offset Coordinate Space:ENiagaraSkelMeshTransforms=::NewEnumerator2
    Skeletal Mesh=$System.Skeletal Mesh Reader
   )
   &FN_Mesh_Attributes(
    Color=&MakeLinearColorFromVectorAndFloat(
     "Vector (RGB)"=&Multiply_Vector(
      A=&MakeVectorFromLinearColorRGB(Color=&self:ScratchDynamicInput_01(
        BlotColor:LinearColor=<rgba 1 0.0743 0.1514 0.1>
        CitadelColor:LinearColor=<rgba 0.0743 0.3828 1 0.1>
        FrothColor:LinearColor=<rgba 0 0 0 1>
        VariantColorIndex:NiagaraInt32=&MakeIntFromFloat(Float=$User.VariantColorIndex)
       ))
      B=&Multiply_VectorByFloat(Vector=<1 1 1>, Float=3)
     )
     "Float (Alpha)"=&Multiply_Float(A=0.4, B=&FN_FrontEndFloatModifier(Gameplay=1, Front End=0.3))
    )
    Create Color Attribute=true
    Module Output=::NewEnumerator1
   ) {module=FN_Mesh_Attributes002}
   &FN_Sprite_Attributes(
    Camera Offset=30
    Color=&MakeLinearColorFromVectorAndFloat(
     "Vector (RGB)"=&MakeVectorFromLinearColorRGB(Color=&self:ScratchDynamicInput_01(
       BlotColor:LinearColor=<rgba 15 0 0 1>
       CitadelColor:LinearColor=<rgba 0 5 15 1>
       FrothColor:LinearColor=<rgba 0 0 0 1>
       VariantColorIndex:NiagaraInt32=&MakeIntFromFloat(Float=$User.VariantColorIndex)
      ))
     "Float (Alpha)"=1.5
    )
    Create Camera Offset Attribute=true
    Create Color Attribute=true
    Module Output:ENiagara_AttributeSamplingApplyOutput=::NewEnumerator1
   )
   &Light_Attributes(
    Color=1
    Color Mode=0
    Falloff Exponent=2
    Light Color:LinearColor=&self:ScratchDynamicInput_01(
     BlotColor:LinearColor=<rgba 15 0 0 1>
     CitadelColor:LinearColor=<rgba 0 5 15 1>
     FrothColor:LinearColor=<rgba 0 0 0 1>
     VariantColorIndex:NiagaraInt32=&MakeIntFromFloat(Float=$User.VariantColorIndex)
    )
    Light Exponent=3
    Light Position Offset:Vector3f=<0 0 30>
    Light Radius=50
    Position=1
    Radius=2
    UsePositionOffset=true
   )
  ]
  rendererClasses [
   &NiagaraMeshRendererProperties
   &NiagaraSpriteRendererProperties
   &NiagaraLightRendererProperties
  ]
  renderers [
   {rendererClass=&NiagaraMeshRendererProperties, rendererIndex=0}
   {rendererClass=&NiagaraMeshRendererProperties, rendererIndex=1}
   {rendererClass=&NiagaraSpriteRendererProperties, rendererIndex=2}
   {rendererClass=&NiagaraLightRendererProperties, rendererIndex=3}
  ]
 }
 HandLight_L {
  emitterUpdateScript [
   /SetVariables(Emitter.Des:NiagaraFloat=&self:ScratchDynamicInput_09(
     Blot:NiagaraFloat=0.5
     Citadel:NiagaraFloat=0.5
     Froth:NiagaraFloat=1
     VariantColorIndex:NiagaraInt32=&MakeIntFromFloat(Float=$User.VariantColorIndex)
    )) {isSetParameters=true}
   /SetVariables(Emitter.HighLightNoisePower:NiagaraFloat=&self:ScratchDynamicInput_09(
     Blot:NiagaraFloat=3
     Citadel:NiagaraFloat=2
     Froth:NiagaraFloat=3
     VariantColorIndex:NiagaraInt32=&MakeIntFromFloat(Float=$User.VariantColorIndex)
    )) {isSetParameters=true}
   &EmitterState(
    Inactive Response=::NewEnumerator1
    Life Cycle Mode=::NewEnumerator1
    Loop Behavior=::NewEnumerator1
    Loop Duration Mode=::NewEnumerator1
    UseLoopDelay=false
   )
   &SpawnBurst_Instantaneous(Spawn Count=4, Use Spawn Probability=false, UseLoopCountLimit=false)
  ]
  particleSpawnScript [
   &PartitionParticles(Number Of Partitions=4)
   &InitializeParticle(Color Mode=::NewEnumerator0, Mesh Scale Mode=::NewEnumerator3, Mesh Uniform Scale=1.3, UsePositionOffset=false, Write Material Random=true)
   &InitialMeshOrientation(Mesh Orientation Mode=::NewEnumerator7, Rotation=<0 -0.25 0>, Use Rotation Vector=true)
   &DynamicMaterialParameters(
    Index 0 Param 1=1.3
    Index 0 Param 2=0.4
    Module.Index0Float0Write=true
    Module.Index0Float1Write=true
    Module.Index0Float2Write=true
    Module.Index0Float3Write=true
   )
   &InitialMeshOrientation(
    Mesh Orientation Mode=::NewEnumerator7
    Rotation=&SelectVectorFromArray(
     Direct Array Index=$Particles.PartitionParticles.Partition
     Vector Selection Array=DataInterface() {propertyValues {FloatData [<0 0.1 0>, <0 -0.1 0>, <0.1 0 0>, <-0.1 0 0>]}}
    )
    Use Rotation Vector=true
   ) {module=InitialMeshOrientation001}
   &CameraOffset(Camera Offset Amount=-30)
  ]
  particleUpdateScript [
   &ParticleState(Kill Particles When Lifetime Has Elapsed=false)
   /SetVariables(Particles.AlphaAll:NiagaraFloat=$System.InterpolateOverTime.InterpolateOverTime_Float.Moving Average) {isSetParameters=true}
   &ScaleColor(
    Scale Alpha=$Particles.AlphaAll
    Scale RGB=&Multiply_VectorByFloat(Vector=<1 1 1>, Float=&FN_FrontEndFloatModifier(Gameplay=1, Front End=0.8))
    ScaleA=true
    ScaleRGB=true
   )
   &FN_Mesh_Attributes(
    Color=&MakeLinearColorFromVectorAndFloat(
     "Vector (RGB)"=&Multiply_Vector(
      A=&MakeVectorFromLinearColorRGB(Color=&self:ScratchDynamicInput_04(
        BlotColor:LinearColor=<rgba 1 0.0146 0 0.4>
        CitadelColor:LinearColor=<rgba 0 0.3333 1 0.4>
        FrothColor:LinearColor=<rgba 0 0 0 1>
        VariantColorIndex:NiagaraInt32=&MakeIntFromFloat(Float=$User.VariantColorIndex)
       ))
      B=<1 1 1>
     )
     "Float (Alpha)"=&Multiply_Float(
      A=&Multiply_Float(A=0.4, B=&FN_FrontEndFloatModifier(Gameplay=1, Front End=0.4))
      B=$Particles.AlphaAll
     )
    )
    Create Color Attribute=true
    Module Output=::NewEnumerator1
   )
   &SkeletalMeshLocation(
    Bone Sampling Mode=::NewEnumerator4
    MaskSpawnGroup=false
    Orientation Sampling=::NewEnumerator1
    Position Sampling=::NewEnumerator2
    Sampled Position Offset=<5 0 0>
    Sampled Position Offset Coordinate Space:ENiagaraSkelMeshTransforms=::NewEnumerator2
    Skeletal Mesh=$System.Skeletal Mesh Reader
   )
   &SkeletalMeshLocation(
    Bone / Socket Index=1
    Bone Sampling Mode=::NewEnumerator4
    MaskSpawnGroup=false
    Orientation Sampling=::NewEnumerator1
    Position Sampling=::NewEnumerator2
    Sampled Position Offset=<-5 0 0>
    Skeletal Mesh=$System.Skeletal Mesh Reader
   ) {module=SkeletalMeshLocation001}
   /SetVariables(Particles.Position01:NiagaraPosition=$Particles.SkeletalMeshLocation.SampledPosition) {isSetParameters=true}
   /SetVariables(Particles.Position02:NiagaraPosition=$Particles.SkeletalMeshLocation001.SampledPosition) {isSetParameters=true}
  ]
  rendererClasses [&NiagaraMeshRendererProperties]
  renderers [
   {rendererClass=&NiagaraMeshRendererProperties, rendererIndex=0}
   {rendererClass=&NiagaraMeshRendererProperties, rendererIndex=1}
   {rendererClass=&NiagaraMeshRendererProperties, rendererIndex=2}
   {rendererClass=&NiagaraMeshRendererProperties, rendererIndex=3}
  ]
 }
 HandLight_R {
  bEnabled=false
  emitterUpdateScript [
   &EmitterState(Life Cycle Mode=::NewEnumerator1, Loop Behavior=::NewEnumerator1, Loop Duration Mode=::NewEnumerator1, UseLoopDelay=false)
   &SpawnBurst_Instantaneous(Spawn Count=4, Use Spawn Probability=false, UseLoopCountLimit=false)
  ]
  particleSpawnScript [
   &PartitionParticles(Number Of Partitions=4)
   &InitializeParticle(Color Mode=::NewEnumerator0, Mesh Scale Mode=::NewEnumerator3, Mesh Uniform Scale=1.3, UsePositionOffset=false, Write Material Random=true)
   &InitialMeshOrientation(Mesh Orientation Mode=::NewEnumerator7, Rotation=<0 -0.25 0>, Use Rotation Vector=true)
   &DynamicMaterialParameters(
    Index 0 Param 1=1.3
    Index 0 Param 2=0.4
    Module.Index0Float0Write=true
    Module.Index0Float1Write=true
    Module.Index0Float2Write=true
    Module.Index0Float3Write=true
   )
   &InitialMeshOrientation(
    Mesh Orientation Mode=::NewEnumerator7
    Rotation=&SelectVectorFromArray(
     Direct Array Index=$Particles.PartitionParticles.Partition
     Vector Selection Array=DataInterface() {propertyValues {FloatData [<0 0.1 0>, <0 -0.1 0>, <0.1 0 0>, <-0.1 0 0>]}}
    )
    Use Rotation Vector=true
   ) {module=InitialMeshOrientation001}
   &CameraOffset(Camera Offset Amount=-30)
  ]
  particleUpdateScript [
   &ParticleState(Kill Particles When Lifetime Has Elapsed=false)
   /SetVariables(Particles.AlphaAll:NiagaraFloat=$System.InterpolateOverTime.InterpolateOverTime_Float.Moving Average) {isSetParameters=true}
   &ScaleColor(
    Scale Alpha=$Particles.AlphaAll
    Scale RGB=&Multiply_VectorByFloat(Vector=<1 1 1>, Float=&FN_FrontEndFloatModifier(Gameplay=1, Front End=0.8))
    ScaleA=true
    ScaleRGB=true
   )
   &FN_Mesh_Attributes(
    Color=&MakeLinearColorFromVectorAndFloat(
     "Vector (RGB)"=&Multiply_Vector(A=&MakeVectorFromLinearColorRGB(Color=<rgba 1 0.0146 0 0.4>), B=<1 1 1>)
     "Float (Alpha)"=&Multiply_Float(
      A=&Multiply_Float(A=0.4, B=&FN_FrontEndFloatModifier(Gameplay=1, Front End=0.4))
      B=$Particles.AlphaAll
     )
    )
    Create Color Attribute=true
    Module Output=::NewEnumerator1
   ) {module=FN_Mesh_Attributes001}
   &SkeletalMeshLocation(
    Bone / Socket Index=1
    Bone Sampling Mode=::NewEnumerator4
    MaskSpawnGroup=false
    Orientation Sampling=::NewEnumerator1
    Sampled Position Offset=<-5 0 0>
    Sampled Position Offset Coordinate Space:ENiagaraSkelMeshTransforms=::NewEnumerator2
    Skeletal Mesh=$System.Skeletal Mesh Reader
   )
  ]
  rendererClasses [&NiagaraMeshRendererProperties]
  renderers [
   {rendererClass=&NiagaraMeshRendererProperties, rendererIndex=0}
   {rendererClass=&NiagaraMeshRendererProperties, rendererIndex=1}
  ]
 }
 HandsFragments {
  emitterUpdateScript [
   /SetVariables(Emitter.CrackColor:LinearColor=<rgba 1.0736 2.8405 5 1>) {isSetParameters=true}
   /SetVariables(Emitter.FresnelEdgeColor:LinearColor=<rgba 1.1782 3.9343 5 1>) {isSetParameters=true}
   &EmitterState(Life Cycle Mode=::NewEnumerator1, Loop Behavior=::NewEnumerator1, Loop Duration Mode=::NewEnumerator1, UseLoopDelay=false)
   &SpawnBurst_Instantaneous(Use Spawn Probability=false, UseLoopCountLimit=false)
  ]
  particleSpawnScript [
   &InitializeParticle(Color Mode=::NewEnumerator0, Mesh Scale Mode=::NewEnumerator3, Mesh Uniform Scale=1.1, UsePositionOffset=false, Write Material Random=true)
   &InitialMeshOrientation(Mesh Orientation Mode=::NewEnumerator7, Use Rotation Vector=true)
   &InitialMeshOrientation(Mesh Orientation Mode=::NewEnumerator7, Use Rotation Vector=true) {module=InitialMeshOrientation001}
   &CameraOffset()
  ]
  particleUpdateScript [
   &ParticleState(Kill Particles When Lifetime Has Elapsed=false)
   &InterpolateOverTime(
    Rate Of Change=8
    Target Float Value:NiagaraFloat=&MakeCustomFloatFromBool(Bool=&SetBoolByFloatComparison(A=$User.ParaGlideLeanAlpha, B=0.7), True Float=1, False Float=0)
   )
   &ScaleMeshSize(Scale Factor=&Multiply_VectorByFloat(
     Vector=<1 1.5 1>
     Float=&Lerp_Float(
      A=0.9
      B=&MakeCustomFloatFromBool(
       Bool=&SetBoolByFloatComparison(A=$Particles.InterpolateOverTime.InterpolateOverTime_Float.Moving Average, B=0.3, Comparison Type=::NewEnumerator6)
       True Float=0
       False Float=1.1
      )
      Alpha=$Particles.InterpolateOverTime.InterpolateOverTime_Float.Moving Average
     )
    ))
   &SkeletalMeshLocation(
    Bone Sampling Mode=::NewEnumerator4
    MaskSpawnGroup=false
    Orientation Sampling=::NewEnumerator1
    Sampled Position Offset Coordinate Space:ENiagaraSkelMeshTransforms=::NewEnumerator2
    Skeletal Mesh=$System.Skeletal Mesh Reader
   )
   &SkeletalMeshLocation(
    Bone / Socket Index=1
    Bone Sampling Mode=::NewEnumerator4
    MaskSpawnGroup=false
    Orientation Sampling=::NewEnumerator1
    Position Sampling=::NewEnumerator2
    Sampled Position Offset Coordinate Space:ENiagaraSkelMeshTransforms=::NewEnumerator2
    Skeletal Mesh=$System.Skeletal Mesh Reader
   ) {module=SkeletalMeshLocation001}
   /SetVariables(Particles.Position01:NiagaraPosition=&AddVectorToPosition(
     Position=$Particles.SkeletalMeshLocation.SampledPosition
     Vector=&AddVectorToPosition(
      Position=&Lerp_Vector(
       A=<0 0 0>
       B=<0 0.5 2.5>
       Alpha=&Waveform(
        Global Amplitude Scale=2
        "Waveform [1]"=::NewEnumerator0
        "[1] Phase Scale"=&RandomRangeFloat(Minimum=1, Maximum=1.2)
        "[1] Frequency"=&RandomRangeFloat(Minimum=0.5, Maximum=0.8)
        "[1] Amplitude Min/Max"=<0 1.5>
       )
      )
      Vector=<6 3 20>
     )
    )) {isSetParameters=true, module=SetVariables_#8}
   /SetVariables(Particles.Position02:NiagaraPosition=&AddVectorToPosition(
     Position=&AddVectorToPosition(
      Position=$Particles.SkeletalMeshLocation.SampledPosition
      Vector=&Lerp_Vector(
       A=<0 0 0>
       B=<0 0.5 2.5>
       Alpha=&Waveform(
        Global Amplitude Scale=2
        "Waveform [1]"=::NewEnumerator1
        "[1] Phase Scale"=&RandomRangeFloat(Minimum=1, Maximum=1.2)
        "[1] Frequency"=&RandomRangeFloat(Minimum=0.5, Maximum=1)
        "[1] Amplitude Min/Max"=<0 1.5>
       )
      )
     )
     Vector=<5 0 15>
    )) {isSetParameters=true, module=SetVariables_#9}
   /SetVariables(Particles.Position03:NiagaraPosition=&AddVectorToPosition(
     Position=&AddVectorToPosition(
      Position=$Particles.SkeletalMeshLocation001.SampledPosition
      Vector=&Lerp_Vector(
       A=<0 0 -1>
       B=<0 0.5 1>
       Alpha=&Waveform(
        Global Amplitude Scale=2
        "Waveform [1]"=::NewEnumerator0
        "[1] Phase Scale"=&RandomRangeFloat(Minimum=1, Maximum=1.2)
        "[1] Frequency"=&RandomRangeFloat(Minimum=0.5, Maximum=1)
        "[1] Amplitude Min/Max"=<0 1.5>
       )
      )
     )
     Vector=<-6 0 15>
    )) {isSetParameters=true, module=SetVariables_#10}
   /SetVariables(Particles.Position04:NiagaraPosition=&AddVectorToPosition(
     Position=$Particles.SkeletalMeshLocation001.SampledPosition
     Vector=&AddVectorToPosition(
      Position=&Lerp_Vector(
       A=<0 0 3>
       B=<0 0.5 5>
       Alpha=&Waveform(
        Global Amplitude Scale=2
        "Waveform [1]"=::NewEnumerator1
        "[1] Phase Scale"=&RandomRangeFloat(Minimum=1, Maximum=1.2)
        "[1] Frequency"=&RandomRangeFloat(Minimum=0.5, Maximum=0.8)
        "[1] Amplitude Min/Max"=<0 1.5>
       )
      )
      Vector=<-6 3 15>
     )
    )) {isSetParameters=true, module=SetVariables_#11}
  ]
  rendererClasses [&NiagaraMeshRendererProperties]
  renderers [
   {rendererClass=&NiagaraMeshRendererProperties, rendererIndex=0}
   {rendererClass=&NiagaraMeshRendererProperties, rendererIndex=1}
   {rendererClass=&NiagaraMeshRendererProperties, rendererIndex=2}
   {rendererClass=&NiagaraMeshRendererProperties, rendererIndex=3}
   {rendererClass=&NiagaraMeshRendererProperties, rendererIndex=4}
   {rendererClass=&NiagaraMeshRendererProperties, rendererIndex=5}
   {rendererClass=&NiagaraMeshRendererProperties, rendererIndex=6}
   {rendererClass=&NiagaraMeshRendererProperties, rendererIndex=7}
   {rendererClass=&NiagaraMeshRendererProperties, rendererIndex=8}
   {rendererClass=&NiagaraMeshRendererProperties, rendererIndex=9}
   {rendererClass=&NiagaraMeshRendererProperties, rendererIndex=10}
   {rendererClass=&NiagaraMeshRendererProperties, rendererIndex=11}
  ]
 }
 Lightning_Small {
  emitterUpdateScript [
   /SetVariables(Emitter.LightningOutline:LinearColor=&self:ScratchDynamicInput_11(
     BlotColor:LinearColor=<rgba 0.3 0 0.01 1>
     CitadelColor:LinearColor=<rgba 0 0.6667 2 1>
     FrothColor:LinearColor=<rgba 0.2 0.2 0.2 1>
     VariantColorIndex:NiagaraInt32=&MakeIntFromFloat(Float=$User.VariantColorIndex)
    )) {isSetParameters=true}
   /SetVariables(Emitter.Des:NiagaraFloat=&self:ScratchDynamicInput_12(
     Blot:NiagaraFloat=0.3
     Citadel:NiagaraFloat=0.3
     Froth:NiagaraFloat=1
     VariantColorIndex:NiagaraInt32=&MakeIntFromFloat(Float=$User.VariantColorIndex)
    )) {isSetParameters=true}
   &EmitterState()
   &SpawnRate(
    SpawnRate=&Lerp_Float(A=10, B=15, Alpha=$System.InterpolateOverTime.InterpolateOverTime_Float.Moving Average)
    Use Spawn Probability=false
   )
  ]
  particleSpawnScript [
   &PartitionParticles()
   &InitializeParticle(
    Color=<rgba 5 5 5 1>
    Lifetime Max=1
    Lifetime Min=0.7
    Lifetime Mode=::NewEnumerator1
    Sprite Rotation Mode=::NewEnumerator2
    Sprite Size Max=<20 30>
    Sprite Size Min=<10 15>
    Sprite Size Mode=::NewEnumerator2
    UsePositionOffset=false
    Write Material Random=true
   )
   &SubUVAnimation(
    Evaluation Type=::NewEnumerator0
    "Sprite Renderer "=DataInterface() {propertyValues {SpriteRenderer=&self:Lightning_Small_0.NiagaraSpriteRendererProperties_0}}
    SubUV Animation Mode:ENiagara_SubUVLookupModeV2=::NewEnumerator2
    UseEndFrame=false
    UseStartFrame=false
   )
   &ShapeLocation(
    Handle Radius=0
    Large Radius=30
    MaskSpawnGroup=false
    Override Seed=false
    Rotation Axis=<0 1 0>
    Rotation Mode=::NewEnumerator1
    Shape Origin=&VectorFromFloat(Value=0)
    Shape Primitive=::NewEnumerator3
    UseOwnerScale=true
   )
   &AddVelocity(
    Velocity=&RandomRangeVector(Minimum=<0 0 200>, Maximum=<0 0 400>)
    Velocity Speed Scale=&Lerp_Float(A=0.5, B=1, Alpha=$System.InterpolateOverTime.InterpolateOverTime_Float.Moving Average)
   )
  ]
  particleUpdateScript [
   /SetVariables(Particles.AlphaAll=1) {isSetParameters=true}
   &SkeletalMeshLocation(
    Bone / Socket Index=$Particles.PartitionParticles.Partition
    Bone Sampling Mode=::NewEnumerator4
    MaskSpawnGroup=false
    Position Sampling=::NewEnumerator1
    Sampled Position Offset=&AddVectorToPosition(
     Position=$Particles.ShapeLocation.ShapePosition
     Vector=&RandomRangeVector(Minimum=<0 0 50>, Maximum=<0 0 150>)
    )
    Sampled Position Offset Coordinate Space:ENiagaraSkelMeshTransforms=::NewEnumerator2
    Skeletal Mesh=$System.Skeletal Mesh Reader
   )
   &ParticleState()
   &ScaleSpriteSize(
    Uniform Curve Scale=&RandomRangeFloat(Minimum=1, Maximum=1.3)
    Uniform Curve Sprite Scale=DataInterface() {propertyValues {Curve {
       defaultValue=0
       keys [
        {interpMode=RCIM_Cubic, tangentMode=RCTM_User, time=0, value=0}
        {arriveTangent=1, interpMode=RCIM_Cubic, leaveTangent=1, time=0.15, value=1.7082}
        {interpMode=RCIM_Cubic, tangentMode=RCTM_User, time=1, value=1}
       ]
      }}}
   )
   &SpriteRotationRate(Rotation Rate=&RandomRangeFloat(Minimum=-150, Maximum=150))
   &Drag(Drag=0.5, Use Linear Drag=true, Use Rotational Drag=true) {enabled=false}
   &SolveForcesAndVelocity(Clamp Velocity=false, Limit Acceleration=false)
   &FN_Mesh_Attributes(
    Color=&MakeLinearColorFromVectorAndFloat(
     "Vector (RGB)"=&Multiply_Vector(
      A=&MakeVectorFromLinearColorRGB(Color=&self:ScratchDynamicInput_10(
        BlotColor:LinearColor=<rgba 1 0.0146 0 0.4>
        CitadelColor:LinearColor=<rgba 0 0.3333 1 0.4>
        FrothColor:LinearColor=<rgba 0 0 0 1>
        VariantColorIndex:NiagaraInt32=&MakeIntFromFloat(Float=$User.VariantColorIndex)
       ))
      B=&Multiply_VectorByFloat(Vector=<1 1 1>, Float=2)
     )
     "Float (Alpha)"=&Multiply_Float(A=&FN_FrontEndFloatModifier(Gameplay=2, Front End=0.4), B=$Particles.AlphaAll)
    )
    Create Color Attribute=true
    Module Output=::NewEnumerator1
   )
   &DynamicMaterialParameters(
    Index 0 Param 4=&FloatFromCurve(FloatCurve=DataInterface() {propertyValues {Curve {defaultValue=1, keys [{tangentMode=RCTM_User, time=0, value=0.3}, {tangentMode=RCTM_User, time=1, value=1}]}}})
    Module.Index0Float0Write=true
    Module.Index0Float1Write=true
    Module.Index0Float2Write=true
    Module.Index0Float3Write=true
   )
   &ScaleColor(
    Scale Alpha=$Particles.AlphaAll
    Scale RGB=&Multiply_VectorByFloat(
     Vector=&Multiply_VectorByFloat(
      Vector=<1 1 1>
      Float=&self:ScratchDynamicInput_13(
       Blot:NiagaraFloat=1
       Citadel:NiagaraFloat=0.6
       Froth:NiagaraFloat=1
       VariantColorIndex:NiagaraInt32=&MakeIntFromFloat(Float=$User.VariantColorIndex)
      )
     )
     Float=&FloatFromCurve(
      FloatCurve=DataInterface() {propertyValues {Curve {keys [
          <at 0 2 RCIM_Cubic>
          {arriveTangent=0.4186, interpMode=RCIM_Cubic, leaveTangent=0.4186, time=0.2181, value=2}
          <at 0.4873 2.204 RCIM_Cubic>
          <at 1 2>
         ]}}}
      Scale Curve=0.7
     )
    )
    ScaleA=true
    ScaleRGB=true
   )
  ]
  rendererClasses [&NiagaraSpriteRendererProperties]
  renderers [
   {rendererClass=&NiagaraSpriteRendererProperties}
   {rendererClass=&NiagaraSpriteRendererProperties, rendererIndex=1}
  ]
 }
 Minimal001 {
  bEnabled=false
  emitterUpdateScript [
   &EmitterState(Life Cycle Mode=::NewEnumerator1, Loop Behavior=::NewEnumerator1, Loop Duration Mode=::NewEnumerator1, UseLoopDelay=false)
   &SpawnBurst_Instantaneous(Use Spawn Probability=false, UseLoopCountLimit=false)
  ]
  particleSpawnScript [
   &InitializeParticle(Color Mode=::NewEnumerator0, Mesh Scale Mode=::NewEnumerator3, Mesh Uniform Scale=1.2, UsePositionOffset=false, Write Material Random=true)
   &InitialMeshOrientation(Mesh Orientation Mode=::NewEnumerator7, Use Rotation Vector=true)
   &InitialMeshOrientation(Mesh Orientation Mode=::NewEnumerator7, Use Rotation Vector=true) {module=InitialMeshOrientation001}
  ]
  particleUpdateScript [
   &ParticleState(Kill Particles When Lifetime Has Elapsed=false)
   &SkeletalMeshLocation(
    Bone Sampling Mode=::NewEnumerator4
    MaskSpawnGroup=false
    Orientation Sampling=::NewEnumerator1
    Sampled Position Offset Coordinate Space:ENiagaraSkelMeshTransforms=::NewEnumerator2
    Skeletal Mesh=$System.Skeletal Mesh Reader
   )
   &SkeletalMeshLocation(
    Bone / Socket Index=1
    Bone Sampling Mode=::NewEnumerator4
    MaskSpawnGroup=false
    Orientation Sampling=::NewEnumerator1
    Position Sampling=::NewEnumerator2
    Sampled Position Offset Coordinate Space:ENiagaraSkelMeshTransforms=::NewEnumerator2
    Skeletal Mesh=$System.Skeletal Mesh Reader
   ) {module=SkeletalMeshLocation001}
   /SetVariables(Particles.Position01:NiagaraPosition=&AddVectorToPosition(
     Position=$Particles.SkeletalMeshLocation.SampledPosition
     Vector=&AddVectorToPosition(
      Position=&Lerp_Vector(
       A=<0 0 0>
       B=<0 2 3>
       Alpha=&Waveform(
        "Waveform [1]"=::NewEnumerator0
        "[1] Phase Scale"=&RandomRangeFloat(Minimum=0.5, Maximum=0.7)
        "[1] Frequency"=&RandomRangeFloat(Minimum=0.5, Maximum=0.8)
        "[1] Amplitude Min/Max"=<0 1.5>
       )
      )
      Vector=<5 0 0>
     )
    )) {isSetParameters=true, module=SetVariables_#8}
   /SetVariables(Particles.Position02:NiagaraPosition=&AddVectorToPosition(
     Position=&AddVectorToPosition(
      Position=$Particles.SkeletalMeshLocation.SampledPosition
      Vector=&Lerp_Vector(
       A=<0 0 0>
       B=<0 2 3>
       Alpha=&Waveform(
        Global Amplitude Scale=&RandomRangeFloat(Minimum=0.5, Maximum=1)
        "Waveform [1]"=::NewEnumerator1
        "[1] Phase Scale"=&RandomRangeFloat(Minimum=0.5, Maximum=0.8)
        "[1] Frequency"=&RandomRangeFloat(Minimum=0.5, Maximum=1)
        "[1] Amplitude Min/Max"=<0 1.5>
       )
      )
     )
     Vector=<5 0 0>
    )) {isSetParameters=true, module=SetVariables_#9}
   /SetVariables(Particles.Position03:NiagaraPosition=&AddVectorToPosition(
     Position=&AddVectorToPosition(Position=$Particles.SkeletalMeshLocation001.SampledPosition, Vector=<0 0 0>)
     Vector=<0 0 0>
    )) {isSetParameters=true, module=SetVariables_#10}
   /SetVariables(Particles.Position04:NiagaraPosition=&AddVectorToPosition(
     Position=$Particles.SkeletalMeshLocation001.SampledPosition
     Vector=&AddVectorToPosition(
      Position=&Lerp_Vector(
       A=<0 0 0>
       B=<0 2 3>
       Alpha=&Waveform(
        "Waveform [1]"=::NewEnumerator0
        "[1] Phase Scale"=&RandomRangeFloat(Minimum=0.5, Maximum=0.7)
        "[1] Frequency"=&RandomRangeFloat(Minimum=0.5, Maximum=0.8)
        "[1] Amplitude Min/Max"=<0 1.5>
       )
      )
      Vector=<5 0 0>
     )
    )) {isSetParameters=true, module=SetVariables_#11}
  ]
  rendererClasses [&NiagaraMeshRendererProperties]
  renderers [
   {rendererClass=&NiagaraMeshRendererProperties, rendererIndex=0}
   {rendererClass=&NiagaraMeshRendererProperties, rendererIndex=1}
   {rendererClass=&NiagaraMeshRendererProperties, rendererIndex=2}
   {rendererClass=&NiagaraMeshRendererProperties, rendererIndex=3}
  ]
 }
 TrailMain {
  emitterUpdateScript [
   /SetVariables(Emitter.Des:NiagaraFloat=&self:ScratchDynamicInput_09(
     Blot:NiagaraFloat=0.5
     Citadel:NiagaraFloat=0.5
     Froth:NiagaraFloat=1
     VariantColorIndex:NiagaraInt32=&MakeIntFromFloat(Float=$User.VariantColorIndex)
    )) {isSetParameters=true}
   /SetVariables(Emitter.HighLightNoisePower:NiagaraFloat=&self:ScratchDynamicInput_09(
     Blot:NiagaraFloat=3
     Citadel:NiagaraFloat=2.5
     Froth:NiagaraFloat=3
     VariantColorIndex:NiagaraInt32=&MakeIntFromFloat(Float=$User.VariantColorIndex)
    )) {isSetParameters=true}
   &EmitterState(
    Inactive Response=::NewEnumerator1
    Life Cycle Mode=::NewEnumerator1
    Loop Behavior=::NewEnumerator1
    Loop Duration Mode=::NewEnumerator1
    UseLoopDelay=false
   )
   &SpawnBurst_Instantaneous(Spawn Count=2, Use Spawn Probability=false, UseLoopCountLimit=false)
  ]
  particleSpawnScript [
   &PartitionParticles()
   &InitializeParticle(Color Mode=::NewEnumerator0, Mesh Scale Mode=::NewEnumerator3, Mesh Uniform Scale=6, UsePositionOffset=false, Write Material Random=true)
   &InitialMeshOrientation(Mesh Orientation Mode=::NewEnumerator7, Rotation=<0 -0.25 0>, Use Rotation Vector=true)
   &CameraOffset(Camera Offset Amount=-20)
  ]
  particleUpdateScript [
   &DynamicMaterialParameters(
    Index 0 Param 1=&Lerp_Float(A=2.5, B=3, Alpha=$User.ParaGlideLeanAlpha)
    Index 0 Param 2=&SelectFloatFromArray(Direct Array Index=$Particles.PartitionParticles.Partition, Float Selection Array=DataInterface() {propertyValues {FloatData [0.3,0.6]}})
    Module.Index0Float0Write=true
    Module.Index0Float1Write=true
    Module.Index0Float2Write=true
    Module.Index0Float3Write=true
   )
   &ParticleState(Kill Particles When Lifetime Has Elapsed=false)
   /SetVariables(Particles.SizeAll:NiagaraFloat=&Lerp_Float(A=0.6, B=1, Alpha=$System.InterpolateOverTime.InterpolateOverTime_Float.Moving Average)) {isSetParameters=true}
   &ScaleMeshSize(Scale Factor=&Multiply_VectorByFloat(Vector=<1 1 1>, Float=$Particles.SizeAll))
   &ScaleColor(
    Scale RGB=&Multiply_VectorByFloat(
     Vector=&Multiply_VectorByFloat(
      Vector=<1 1 1>
      Float=&FN_PickFloatDependOnPlatform(
       TargetPlatformActive=1
       TargetPlatformNoActive=&self:ScratchDynamicInput_13(
        Blot:NiagaraFloat=1
        Citadel:NiagaraFloat=0.6
        Froth:NiagaraFloat=1
        VariantColorIndex:NiagaraInt32=&MakeIntFromFloat(Float=$User.VariantColorIndex)
       )
      )
     )
     Float=&FN_FrontEndFloatModifier(
      Gameplay=1
      Front End=&self:ScratchDynamicInput_13(
       Blot:NiagaraFloat=1
       Citadel:NiagaraFloat=0.5
       Froth:NiagaraFloat=1
       VariantColorIndex:NiagaraInt32=&MakeIntFromFloat(Float=$User.VariantColorIndex)
      )
     )
    )
    ScaleA=true
    ScaleRGB=true
   )
   &FN_Mesh_Attributes(
    Color=&MakeLinearColorFromVectorAndFloat(
     "Vector (RGB)"=&Multiply_Vector(
      A=&MakeVectorFromLinearColorRGB(Color=&self:ScratchDynamicInput_04(
        BlotColor:LinearColor=<rgba 1 0.0146 0 0.4>
        CitadelColor:LinearColor=<rgba 0 0.3333 1 0.4>
        FrothColor:LinearColor=<rgba 0 0 0 1>
        VariantColorIndex:NiagaraInt32=&MakeIntFromFloat(Float=$User.VariantColorIndex)
       ))
      B=<1 1 1>
     )
     "Float (Alpha)"=&Multiply_Float(
      A=&Multiply_Float(
       A=0.4
       B=&self:ScratchDynamicInput_13(
        Blot:NiagaraFloat=1
        Citadel:NiagaraFloat=&FN_FrontEndFloatModifier(
         Gameplay=&FNTimeOfDayScalar(Night Scale=1, Morning Scale=1, Day Scale=2, Evening Scale=1)
         Front End=3
        )
        Froth:NiagaraFloat=&FN_FrontEndFloatModifier(Gameplay=1.5, Front End=10)
        VariantColorIndex:NiagaraInt32=&MakeIntFromFloat(Float=$User.VariantColorIndex)
       )
      )
      B=&FN_FrontEndFloatModifier(Gameplay=1, Front End=0.2)
     )
    )
    Create Color Attribute=true
    Module Output=::NewEnumerator1
   )
   &SkeletalMeshLocation(
    Bone / Socket Index=$Particles.PartitionParticles.Partition
    Bone Sampling Mode=::NewEnumerator4
    MaskSpawnGroup=false
    Orientation Sampling=::NewEnumerator1
    Sampled Position Offset=&SelectVectorFromArray(
     Direct Array Index=$Particles.PartitionParticles.Partition
     Vector Selection Array=DataInterface() {propertyValues {FloatData [<5 0 0>, <-5 0 0>]}}
    )
    Sampled Position Offset Coordinate Space:ENiagaraSkelMeshTransforms=::NewEnumerator2
    Skeletal Mesh=$System.Skeletal Mesh Reader
   )
  ]
  rendererClasses [&NiagaraMeshRendererProperties]
  renderers [
   {rendererClass=&NiagaraMeshRendererProperties, rendererIndex=0}
   {rendererClass=&NiagaraMeshRendererProperties, rendererIndex=1}
  ]
 }
 sparks {
  simTarget=GPUComputeSim
  emitterUpdateScript [
   &EmitterState(Inactive Response=::NewEnumerator1, Life Cycle Mode=::NewEnumerator1, Loop Duration=4.6, UseLoopDelay=false)
   &SpawnRate(
    Spawn Probability=$System.InterpolateOverTime.InterpolateOverTime_Float.Moving Average
    SpawnRate=&FN_FrontEndFloatModifier(Gameplay=100, Front End=20)
    Use Spawn Probability=true
   )
  ]
  particleSpawnScript [
   &InitializeParticle(
    Color=&self:ScratchDynamicInput_01(
     BlotColor:LinearColor=<rgba 50 0 8.3333 1>
     CitadelColor:LinearColor=<rgba 0 16.6667 50 1>
     FrothColor:LinearColor=<rgba 0 0 0 100>
     VariantColorIndex:NiagaraInt32=&MakeIntFromFloat(Float=$User.VariantColorIndex)
    )
    Lifetime Max=0.5
    Lifetime Min=0.4
    Lifetime Mode=::NewEnumerator1
    Ribbon Width=40
    Ribbon Width Mode=::NewEnumerator1
    Sprite Size Max=<1 6>
    Sprite Size Min=<1 10>
    Sprite Size Mode=::NewEnumerator2
    UsePositionOffset=false
    Write Material Random=true
   )
   &ShapeLocation(
    Box Size=<100 160 50>
    MaskSpawnGroup=false
    Offset=<0 0 100>
    Offset Mode=::NewEnumerator0
    Override Seed=false
    Shape Primitive=::NewEnumerator2
    Surface Only=false
    UseOwnerScale=true
   )
   &AddVelocity(
    Module.Use Falloff Distance=false
    Origin Offset=<0 0 -30>
    Override Seed=false
    Velocity Mode=::NewEnumerator1
    Velocity Speed=&RandomRangeFloat(Minimum=400, Maximum=700)
   )
   /SetVariables(Particles.SizeAll:NiagaraFloat=&FN_FrontEndFloatModifier(Gameplay=1.5, Front End=1)) {isSetParameters=true}
  ]
  particleUpdateScript [
   &ParticleState()
   &ScaleSpriteSize(Scale Factor=&Vector2DFromFloat(Value=&FloatFromCurve(
      FloatCurve=DataInterface() {propertyValues {Curve {keys [
          <at 0 0 RCIM_Cubic>
          {interpMode=RCIM_Cubic, leaveTangent=-3.0296, tangentMode=RCTM_Break, time=0.2, value=1.2}
          {arriveTangent=-1.8107, interpMode=RCIM_Cubic, leaveTangent=-1.8107, tangentMode=RCTM_User, time=1, value=0.2}
         ]}}}
      Scale Curve=&Multiply_Float(A=0.4, B=$Particles.SizeAll)
     ))) {module=ScaleSpriteSize001}
   &SpriteRotationRate(Rotation Rate=&RandomRangeFloat(Minimum=-70, Maximum=70))
   /SetVariables(Particles.Glow03_Color=&MakeLinearColorFromVectorAndFloat(
     "Vector (RGB)"=&MakeVectorFromLinearColorRGB(Color=&MultiplyLinearColor(
       Linear Color=&self:ScratchDynamicInput_01(
        BlotColor:LinearColor=<rgba 1 0.0112 0.0112 1>
        CitadelColor:LinearColor=<rgba 0.0112 0.3408 1 1>
        FrothColor:LinearColor=<rgba 0 0 0 1>
        VariantColorIndex:NiagaraInt32=&MakeIntFromFloat(Float=$User.VariantColorIndex)
       )
       Scale Factor=&MakeVector4_FromFloat(Float=1)
      ))
     "Float (Alpha)"=&Multiply_Float(
      A=&Multiply_Float(
       A=0.1
       B=&self:ScratchDynamicInput_13(
        Blot:NiagaraFloat=1
        Citadel:NiagaraFloat=1
        Froth:NiagaraFloat=1
        VariantColorIndex:NiagaraInt32=&MakeIntFromFloat(Float=$User.VariantColorIndex)
       )
      )
      B=&FN_FrontEndFloatModifier(Gameplay=1, Front End=0.7)
     )
    )) {isSetParameters=true}
   /SetVariables(Particles.Glow03_Size=&Multiply_Vector2D_ByFloat(
     Float=&FloatFromCurve(
      FloatCurve=DataInterface() {propertyValues {Curve {keys [
          <at 0 0 RCIM_Cubic>
          <at 0.2 1.3 RCIM_Cubic>
          {arriveTangent=-0.1423, interpMode=RCIM_Cubic, leaveTangent=-0.1423, tangentMode=RCTM_User, time=1, value=0}
         ]}}}
      Scale Curve=&Multiply_Float(A=6, B=$Particles.SizeAll)
     )
     Vector2D=<1 2>
    )) {isSetParameters=true}
   &DynamicMaterialParameters(
    Index 0 Param 1=&MakeCustomFloatFromBool(Bool=&RandomBool(), True Float=0, False Float=0.75)
    Module.Index0Float0Write=true
    Module.Index0Float1Write=true
    Module.Index0Float2Write=true
    Module.Index0Float3Write=true
   )
   &Drag(Drag=&RandomRangeFloat(Minimum=2, Maximum=3), Use Linear Drag=true, Use Rotational Drag=true)
   &SolveForcesAndVelocity(Clamp Velocity=false, Limit Acceleration=false)
   &ScaleSpriteSizeBySpeed(Max Scale Factor=<1 2>, Min Scale Factor=<0.5 0.5>)
   &CameraOffset(Camera Offset Amount=-10) {module=CameraOffset001}
  ]
  rendererClasses [&NiagaraSpriteRendererProperties]
  renderers [
   {rendererClass=&NiagaraSpriteRendererProperties}
   {rendererClass=&NiagaraSpriteRendererProperties, rendererIndex=1}
  ]
 }
}
summary {
 systemName=NS_Contrail_SheerWill
 emitters [
  {emitterName=TrailMain, rendererClasses [&NiagaraMeshRendererProperties]}
  {emitterName=HandLight_L, rendererClasses [&NiagaraMeshRendererProperties]}
  {bEnabled=false, emitterName=HandLight_R, rendererClasses [&NiagaraMeshRendererProperties]}
  {
   emitterName=HandCone
   rendererClasses [
    &NiagaraMeshRendererProperties
    &NiagaraSpriteRendererProperties
    &NiagaraLightRendererProperties
   ]
  }
  {emitterName=Lightning_Small, rendererClasses [&NiagaraSpriteRendererProperties]}
  {emitterName=BottomFragment, rendererClasses [&NiagaraMeshRendererProperties]}
  {emitterName=BottomEnergy, rendererClasses [&NiagaraMeshRendererProperties]}
  {emitterName=DarkTrail, rendererClasses [&NiagaraMeshRendererProperties]}
  {emitterName=sparks, simTarget=GPUComputeSim, rendererClasses [&NiagaraSpriteRendererProperties]}
  {emitterName=HandsFragments, rendererClasses [&NiagaraMeshRendererProperties]}
  {bEnabled=false, emitterName=Minimal001, rendererClasses [&NiagaraMeshRendererProperties]}
 ]
 userVariables [
  User.bIsFrontEnd {type=NiagaraBool}
  User.bIsFrontEndPreview {type=NiagaraBool}
  User.ParaGlideLeanAlpha {type=NiagaraFloat, defaultValue=1, description=""}
  User.VariantColorIndex {type=NiagaraFloat, defaultValue=0, description=""}
 ]
}
